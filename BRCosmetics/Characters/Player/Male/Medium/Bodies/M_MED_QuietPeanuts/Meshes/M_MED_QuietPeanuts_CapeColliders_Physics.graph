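# Physics asset of cape colliders for the M_MED_QuietPeanuts male medium player cosmetic body, by its name serving as collision for the cape.
# One capsule per body covers the pelvis, spine, neck, head, clavicles, arms with upper twist bones, hands, legs and feet.
# An extra large capsule on dyn_turtleShell_01, constrained to spine_04, adds the back shell's volume.
# Constraints chain the bodies along the skeleton, with hip, clavicle, neck, head and pelvis-to-spine joints set free.
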PhysicsAsset /BRCosmetics/Characters/Player/Male/Medium/Bodies/M_MED_QuietPeanuts/Meshes/M_MED_QuietPeanuts_CapeColliders_Physics
bodies {
 calf_l [{shapeType=Capsule, length=25.0626, radius=10.4081, center=<-16.7112 -1.7509 -0.2582>, rotation=<rot 0 -82.9766 -88.3985>}]
 calf_r [{shapeType=Capsule, length=24.1191, radius=10.0596, center=<15.5918 2.0258 0.2933>, rotation=<rot 0 -81.9487 -88.5448>}]
 clavicle_l [{shapeType=Capsule, length=12, radius=12, shapeName=clavicle_l_capsule, center=<9 0 0>, rotation=<rot 90 0 0>}]
 clavicle_r [{shapeType=Capsule, length=12, radius=12, shapeName=clavicle_r_capsule, center=<-9 0 0>, rotation=<rot 90 0 0>}]
 dyn_turtleShell_01 [{shapeType=Capsule, length=22, radius=24.869, center=<-0.0001 14.5664 1.1448>, rotation=<rot 0 -90 0>}]
 foot_l [{shapeType=Capsule, length=17, radius=11, center=<-2 10 -1>, rotation=<rot 90 90 0>}]
 foot_r [{shapeType=Capsule, length=17, radius=11, center=<2 -10 1>, rotation=<rot 90 90 0>}]
 hand_l [{shapeType=Capsule, length=8.5736, radius=6.7687, center=<7.5211 -2.1463 0.1105>, rotation=<rot 0 89.2927 -75.1087>}]
 hand_r [{shapeType=Capsule, length=8.5736, radius=6.7688, center=<-7.5211 2.1463 -0.1104>, rotation=<rot 0 89.2921 -75.1106>}]
 head [{shapeType=Capsule, length=10, radius=10, shapeName=head_capsule, center=<5 3 0>, rotation=<rot 90 90 0>}]
 lowerarm_l [{shapeType=Capsule, length=17.4785, radius=8.5572, center=<6.9085 -1.1793 0.3424>, rotation=<rot 0 99.7209 -85.8822>}]
 lowerarm_r [{shapeType=Capsule, length=17.4785, radius=8.5576, center=<-6.9086 1.1795 -0.3421>, rotation=<rot 0 99.7238 -85.8852>}]
 neck_01 [{shapeType=Capsule, length=10, radius=10, shapeName=neck_01_capsule, center=<5 0 0>, rotation=<rot 90 0 0>}]
 neck_02 [{shapeType=Capsule, length=10, radius=10, shapeName=neck_02_capsule, center=<5 0 0>, rotation=<rot 90 0 0>}]
 pelvis [{shapeType=Capsule, length=16.7519, radius=16.2085, center=<0 -2 0>}]
 spine_01 [{shapeType=Capsule, length=14, radius=16, shapeName=spine_01_capsule, center=<0 -5 0>}]
 spine_02 [{shapeType=Capsule, length=14, radius=20, center=<2.8382 -6 0>}]
 spine_03 [{shapeType=Capsule, length=14, radius=22, center=<7 -6 0>}]
 spine_04 [{shapeType=Capsule, length=14, radius=21, center=<8 -6 0>}]
 spine_05 [{shapeType=Capsule, length=18, radius=18, center=<6.0609 -4 0>}]
 thigh_l [{shapeType=Capsule, length=31.159, radius=12.3242, shapeName=thigh_l_capsule, center=<-19.2304 0 0>, rotation=<rot 90 0 0>}]
 thigh_r [{shapeType=Capsule, length=31.159, radius=12.3242, shapeName=thigh_l_capsule, center=<19.2304 0 0>, rotation=<rot 90 0 180>}]
 upperarm_l [{shapeType=Capsule, length=12.7852, radius=8.577, center=<16.4154 -0.6561 -0.2854>, rotation=<rot 15.5235 -90 -86.1103>}]
 upperarm_r [{shapeType=Capsule, length=12.7868, radius=8.5761, center=<-16.414 0.6549 0.286>, rotation=<rot 15.5604 -90 -86.1071>}]
 upperarm_twist_01_l [{shapeType=Capsule, length=11.5025, radius=10.518, center=<5.4333 -0.6594 -0.4788>, rotation=<rot 40.5687 -90 -21.8426>}]
 upperarm_twist_01_r [{shapeType=Capsule, length=11.495, radius=10.4991, center=<-5.4004 0.6805 0.4815>, rotation=<rot 41.1292 -90 -21.589>}]
}
constraints [
 {bone1Name=spine_02, bone2Name=spine_01, swing1LimitDegrees=45, swing2LimitDegrees=45, twistLimitDegrees=45}
 {bone1Name=spine_03, bone2Name=spine_02, swing1LimitDegrees=45, swing2LimitDegrees=45, twistLimitDegrees=45}
 {bone1Name=spine_04, bone2Name=spine_03, swing1LimitDegrees=45, swing2LimitDegrees=45, twistLimitDegrees=45}
 {bone1Name=spine_05, bone2Name=spine_04, swing1LimitDegrees=45, swing2LimitDegrees=45, twistLimitDegrees=45}
 {bone1Name=upperarm_l, bone2Name=clavicle_l, swing1LimitDegrees=45, swing2LimitDegrees=45, twistLimitDegrees=45}
 {bone1Name=lowerarm_l, bone2Name=upperarm_l, swing1LimitDegrees=45, swing2LimitDegrees=45, twistLimitDegrees=45}
 {bone1Name=hand_l, bone2Name=lowerarm_l, swing1LimitDegrees=45, swing2LimitDegrees=45, twistLimitDegrees=45}
 {bone1Name=upperarm_twist_01_l, bone2Name=upperarm_l, swing1LimitDegrees=45, swing2LimitDegrees=45, twistLimitDegrees=45}
 {bone1Name=upperarm_r, bone2Name=clavicle_r, swing1LimitDegrees=45, swing2LimitDegrees=45, twistLimitDegrees=45}
 {bone1Name=lowerarm_r, bone2Name=upperarm_r, swing1LimitDegrees=45, swing2LimitDegrees=45, twistLimitDegrees=45}
 {bone1Name=hand_r, bone2Name=lowerarm_r, swing1LimitDegrees=45, swing2LimitDegrees=45, twistLimitDegrees=45}
 {bone1Name=upperarm_twist_01_r, bone2Name=upperarm_r, swing1LimitDegrees=45, swing2LimitDegrees=45, twistLimitDegrees=45}
 {bone1Name=dyn_turtleShell_01, bone2Name=spine_04, swing1LimitDegrees=45, swing2LimitDegrees=45, twistLimitDegrees=45}
 {bone1Name=calf_l, bone2Name=thigh_l, swing1LimitDegrees=45, swing2LimitDegrees=45, twistLimitDegrees=45}
 {bone1Name=foot_l, bone2Name=calf_l, swing1LimitDegrees=45, swing2LimitDegrees=45, twistLimitDegrees=45}
 {bone1Name=calf_r, bone2Name=thigh_r, swing1LimitDegrees=45, swing2LimitDegrees=45, twistLimitDegrees=45}
 {bone1Name=foot_r, bone2Name=calf_r, swing1LimitDegrees=45, swing2LimitDegrees=45, twistLimitDegrees=45}
 {
  bone1Name=thigh_l
  bone2Name=pelvis
  swing1LimitDegrees=45
  swing1Motion=Free
  swing2LimitDegrees=45
  swing2Motion=Free
  twistLimitDegrees=45
  twistMotion=Free
 }
 {
  bone1Name=thigh_r
  bone2Name=pelvis
  swing1LimitDegrees=45
  swing1Motion=Free
  swing2LimitDegrees=45
  swing2Motion=Free
  twistLimitDegrees=45
  twistMotion=Free
 }
 {
  bone1Name=clavicle_l
  bone2Name=spine_05
  swing1LimitDegrees=45
  swing1Motion=Free
  swing2LimitDegrees=45
  swing2Motion=Free
  twistLimitDegrees=45
  twistMotion=Free
 }
 {
  bone1Name=clavicle_r
  bone2Name=spine_05
  swing1LimitDegrees=45
  swing1Motion=Free
  swing2LimitDegrees=45
  swing2Motion=Free
  twistLimitDegrees=45
  twistMotion=Free
 }
 {
  bone1Name=neck_01
  bone2Name=spine_05
  swing1LimitDegrees=45
  swing1Motion=Free
  swing2LimitDegrees=45
  swing2Motion=Free
  twistLimitDegrees=45
  twistMotion=Free
 }
 {
  bone1Name=head
  bone2Name=neck_02
  swing1LimitDegrees=45
  swing1Motion=Free
  swing2LimitDegrees=45
  swing2Motion=Free
  twistLimitDegrees=45
  twistMotion=Free
 }
 {
  bone1Name=neck_02
  bone2Name=neck_01
  swing1LimitDegrees=45
  swing1Motion=Free
  swing2LimitDegrees=45
  swing2Motion=Free
  twistLimitDegrees=45
  twistMotion=Free
 }
 {
  bone1Name=spine_01
  bone2Name=pelvis
  swing1LimitDegrees=45
  swing1Motion=Free
  swing2LimitDegrees=45
  swing2Motion=Free
  twistLimitDegrees=45
  twistMotion=Free
 }
]
