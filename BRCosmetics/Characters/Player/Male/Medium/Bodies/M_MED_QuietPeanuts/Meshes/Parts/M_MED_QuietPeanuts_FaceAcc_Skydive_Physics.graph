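# Physics asset for the skydive variant of the QuietPeanuts male character's face accessory, a headband with two dangling strands.
# Left and right five-capsule chains hang from zero-size master bodies linked to the head, which is constrained to the upper spine.
# Chain joints leave swing free with a small twist angle; the other joints are fully free in rotation.
# Spine, neck, head and a large turtle-shell capsule provide body volumes.
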PhysicsAsset /BRCosmetics/Characters/Player/Male/Medium/Bodies/M_MED_QuietPeanuts/Meshes/Parts/M_MED_QuietPeanuts_FaceAcc_Skydive_Physics
bodies {
 HeadBand_MSTR [{shapeType=Capsule, length=0, radius=0, shapeName=HeadBand_MSTR_capsule}]
 dyn_HeadBand_l_01_jnt [{shapeType=Capsule, length=5, radius=3, shapeName=dyn_HeadBand_l_01_jnt_capsule, center=<6 0 0>, rotation=<rot 90 0 0>}]
 dyn_HeadBand_l_02_jnt [{shapeType=Capsule, length=5, radius=3, shapeName=dyn_HeadBand_l_02_jnt_capsule, center=<6 0 0>, rotation=<rot 90 0 0>}]
 dyn_HeadBand_l_03_jnt [{shapeType=Capsule, length=5, radius=3, shapeName=dyn_HeadBand_l_03_jnt_capsule, center=<6 0 0>, rotation=<rot 90 0 0>}]
 dyn_HeadBand_l_04_jnt [{shapeType=Capsule, length=5, radius=3, shapeName=dyn_HeadBand_l_04_jnt_capsule, center=<6 0 0>, rotation=<rot 90 0 0>}]
 dyn_HeadBand_l_05_jnt [{shapeType=Capsule, length=5, radius=3, shapeName=dyn_HeadBand_l_05_jnt_capsule, center=<6 0 0>, rotation=<rot 90 0 0>}]
 dyn_HeadBand_l_MSTR [{shapeType=Capsule, length=0, radius=0, shapeName=dyn_HeadBand_l_MSTR_capsule}]
 dyn_HeadBand_r_01_jnt [{shapeType=Capsule, length=5, radius=3, shapeName=dyn_HeadBand_r_01_jnt_capsule, center=<6 0 0>, rotation=<rot 90 0 0>}]
 dyn_HeadBand_r_02_jnt [{shapeType=Capsule, length=5, radius=3, shapeName=dyn_HeadBand_r_02_jnt_capsule, center=<6 0 0>, rotation=<rot 90 0 0>}]
 dyn_HeadBand_r_03_jnt [{shapeType=Capsule, length=5, radius=3, shapeName=dyn_HeadBand_r_03_jnt_capsule, center=<6 0 0>, rotation=<rot 90 0 0>}]
 dyn_HeadBand_r_04_jnt [{shapeType=Capsule, length=5, radius=3, shapeName=dyn_HeadBand_r_04_jnt_capsule, center=<6 0 0>, rotation=<rot 90 0 0>}]
 dyn_HeadBand_r_05_jnt [{shapeType=Capsule, length=5, radius=3, shapeName=dyn_HeadBand_r_05_jnt_capsule, center=<6 0 0>, rotation=<rot 90 0 0>}]
 dyn_HeadBand_r_MSTR [{shapeType=Capsule, length=0, radius=0, shapeName=dyn_HeadBand_r_MSTR_capsule}]
 dyn_turtleShell_01 [{shapeType=Capsule, length=14, radius=20, shapeName=dyn_turtleShell_01_capsule, center=<0 10.942 11.9676>, rotation=<rot 90 0 0>}]
 head [{shapeType=Capsule, length=5, radius=8, shapeName=head_capsule, center=<0 3 0>, rotation=<rot 90 0 0>}]
 neck_01 [{shapeType=Capsule, length=8, radius=4.5, shapeName=neck_01_capsule, center=<3.6683 -0.2876 0>, rotation=<rot 90 0 0>}]
 spine_03 [{shapeType=Capsule, length=15, radius=10, shapeName=spine_04_capsule, center=<0.229 2 0>, rotation=<rot 90 0 10.0001>}]
 spine_04 [{shapeType=Capsule, length=24, radius=14, shapeName=spine_04_capsule, center=<0 -2.0219 0>}]
 spine_05 [{shapeType=Capsule, length=24, radius=14, shapeName=spine_05_capsule, center=<2.6608 -2.6502 0>}]
}
constraints [
 {
  bone1Name=head
  bone2Name=spine_05
  swing1LimitDegrees=45
  swing1Motion=Free
  swing2LimitDegrees=45
  swing2Motion=Free
  twistLimitDegrees=45
  twistMotion=Free
 }
 {
  bone1Name=spine_05
  bone2Name=spine_04
  swing1LimitDegrees=45
  swing1Motion=Free
  swing2LimitDegrees=45
  swing2Motion=Free
  twistLimitDegrees=45
  twistMotion=Free
 }
 {
  bone1Name=HeadBand_MSTR
  bone2Name=head
  swing1LimitDegrees=45
  swing1Motion=Free
  swing2LimitDegrees=45
  swing2Motion=Free
  twistLimitDegrees=45
  twistMotion=Free
 }
 {
  bone1Name=dyn_HeadBand_l_MSTR
  bone2Name=HeadBand_MSTR
  swing1LimitDegrees=45
  swing1Motion=Free
  swing2LimitDegrees=45
  swing2Motion=Free
  twistLimitDegrees=45
  twistMotion=Free
 }
 {
  bone1Name=dyn_HeadBand_l_01_jnt
  bone2Name=dyn_HeadBand_l_MSTR
  swing1LimitDegrees=11
  swing1Motion=Free
  swing2LimitDegrees=35
  swing2Motion=Free
  twistLimitDegrees=5
 }
 {
  bone1Name=dyn_HeadBand_l_02_jnt
  bone2Name=dyn_HeadBand_l_01_jnt
  swing1LimitDegrees=11
  swing1Motion=Free
  swing2LimitDegrees=35
  swing2Motion=Free
  twistLimitDegrees=5
 }
 {
  bone1Name=dyn_HeadBand_l_03_jnt
  bone2Name=dyn_HeadBand_l_02_jnt
  swing1LimitDegrees=11
  swing1Motion=Free
  swing2LimitDegrees=35
  swing2Motion=Free
  twistLimitDegrees=5
 }
 {
  bone1Name=dyn_HeadBand_l_04_jnt
  bone2Name=dyn_HeadBand_l_03_jnt
  swing1LimitDegrees=11
  swing1Motion=Free
  swing2LimitDegrees=35
  swing2Motion=Free
  twistLimitDegrees=5
 }
 {
  bone1Name=dyn_HeadBand_l_05_jnt
  bone2Name=dyn_HeadBand_l_04_jnt
  swing1LimitDegrees=11
  swing1Motion=Free
  swing2LimitDegrees=35
  swing2Motion=Free
  twistLimitDegrees=5
 }
 {
  bone1Name=dyn_HeadBand_r_MSTR
  bone2Name=HeadBand_MSTR
  swing1LimitDegrees=45
  swing1Motion=Free
  swing2LimitDegrees=45
  swing2Motion=Free
  twistLimitDegrees=45
  twistMotion=Free
 }
 {
  bone1Name=dyn_HeadBand_r_01_jnt
  bone2Name=dyn_HeadBand_r_MSTR
  swing1LimitDegrees=11
  swing1Motion=Free
  swing2LimitDegrees=35
  swing2Motion=Free
  twistLimitDegrees=5
 }
 {
  bone1Name=dyn_HeadBand_r_02_jnt
  bone2Name=dyn_HeadBand_r_01_jnt
  swing1LimitDegrees=11
  swing1Motion=Free
  swing2LimitDegrees=35
  swing2Motion=Free
  twistLimitDegrees=5
 }
 {
  bone1Name=dyn_HeadBand_r_03_jnt
  bone2Name=dyn_HeadBand_r_02_jnt
  swing1LimitDegrees=11
  swing1Motion=Free
  swing2LimitDegrees=35
  swing2Motion=Free
  twistLimitDegrees=5
 }
 {
  bone1Name=dyn_HeadBand_r_04_jnt
  bone2Name=dyn_HeadBand_r_03_jnt
  swing1LimitDegrees=11
  swing1Motion=Free
  swing2LimitDegrees=35
  swing2Motion=Free
  twistLimitDegrees=5
 }
 {
  bone1Name=dyn_HeadBand_r_05_jnt
  bone2Name=dyn_HeadBand_r_04_jnt
  swing1LimitDegrees=11
  swing1Motion=Free
  swing2LimitDegrees=35
  swing2Motion=Free
  twistLimitDegrees=5
 }
 {
  bone1Name=dyn_turtleShell_01
  bone2Name=spine_04
  swing1LimitDegrees=45
  swing1Motion=Free
  swing2LimitDegrees=45
  swing2Motion=Free
  twistLimitDegrees=45
  twistMotion=Free
 }
]
