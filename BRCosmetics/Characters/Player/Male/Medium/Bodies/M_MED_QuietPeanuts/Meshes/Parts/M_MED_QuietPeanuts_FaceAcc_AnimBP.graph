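# Post-process-style AnimBP for the QuietPeanuts face accessory, animating the headband and its two tassels on the parent mesh's copied pose.
# Control Rig feeds RigidBody branches whose physics assets follow skydive state and matching backbling/pickaxe cosmetic tags, with an AnimDynamics chain fallback when rigid bodies are disabled.
# Updates drive wind-like gravity overrides, skydive headband tilt and noise-driven tassel sway (ground and fallback branches), weakened in the front end.
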AnimBlueprint /BRCosmetics/Characters/Player/Male/Medium/Bodies/M_MED_QuietPeanuts/Meshes/Parts/M_MED_QuietPeanuts_FaceAcc_AnimBP
target_skeleton=&/M_MED_QuietPeanuts_FaceAcc_Skeleton
graphs [
 AnimGraph {nodes [
   n1 = Control Rig(Source=Copy Pose From Mesh() {bUseAttachedParent=true}) {
    alphaBoolBlend {blendOption=Linear}
    controlRigAssetReference {blueprintRigClass=&/M_MED_QuietPeanuts_FaceAcc_CtrlRig.M_MED_QuietPeanuts_FaceAcc_CtrlRig_C}
   }
   n2 = RigidBody!(
    ComponentPose=Local To Component(LocalPose=@n1)
    Alpha
    OverrideWorldGravity:Vector=Get GravityOverride_tessel() -> (GravityOverride_tessel:Vector)
    ExternalForce
   ) -> (Pose) {
    bClampLinearTranslationLimitToRefPose=true
    bOverrideWorldGravity=true
    overridePhysicsAsset=&/M_MED_QuietPeanuts_FaceAcc_Backbling_Physics
    simulationSpace=BaseBoneSpace
    alphaBoolBlend {blendOption=Linear}
    baseBoneRef {boneName=neck_01}
    componentAppliedLinearAccClamp=<3000 3000 3000>
    componentLinearAccScale {x=0.1}
    componentLinearVelScale {x=0.1}
    overrideWorldGravity {z=-980}
    simSpaceSettings {
     maxAngularAcceleration=1000
     maxAngularVelocity=3
     maxLinearAcceleration=700
     maxLinearVelocity=500
     velocityScaleZ=0
     worldAlpha=1
     externalLinearDragV=<0.5 0.5>
    }
   }
   n3 = "Transform (Modify) Bone - Bone: HeadBand_MSTR"(ComponentPose=@n2, Alpha=Get Is Skydiving Alpha() -> (Is Skydiving Alpha:double), Translation="0,0,0", Rotation="20,0,0", Scale="1,1,1") {
    rotationMode=BMM_Additive
    rotationSpace=BCS_BoneSpace
    translationSpace=BCS_BoneSpace
    alphaBoolBlend {blendOption=Linear}
    alphaScaleBiasClamp {interpSpeedDecreasing=2, interpSpeedIncreasing=2}
    boneToModify {boneName=HeadBand_MSTR}
    rotation {pitch=20}
   }
   n4 = Control Rig(Source=Copy Pose From Mesh() {bUseAttachedParent=true}) {
    alphaBoolBlend {blendOption=Linear}
    controlRigAssetReference {blueprintRigClass=&M_MED_QuietPeanuts_FaceAcc_CtrlRig_C}
   }
   n5 = RigidBody(ComponentPose=Local To Component(LocalPose=@n4), OverrideWorldGravity=Get GravityOverride_tessel()) {
    bClampLinearTranslationLimitToRefPose=true
    bOverrideWorldGravity=true
    overridePhysicsAsset=&/M_MED_QuietPeanuts_FaceAcc_Skydive_Physics
    simulationSpace=BaseBoneSpace
    alphaBoolBlend {blendOption=Linear}
    baseBoneRef {boneName=neck_01}
    componentAppliedLinearAccClamp=<3000 3000 3000>
    componentLinearAccScale {x=0.1}
    componentLinearVelScale {x=0.1}
    overrideWorldGravity {z=-980}
    simSpaceSettings {
     maxAngularAcceleration=1000
     maxAngularVelocity=3
     maxLinearAcceleration=700
     maxLinearVelocity=500
     velocityScaleZ=0
     worldAlpha=1
     externalLinearDragV=<0.5 0.5>
    }
   }
   n6 = "Transform (Modify) Bone - Bone: HeadBand_MSTR"(ComponentPose=@n5, Alpha=Get Is Skydiving Alpha(), Translation="0,0,0", Rotation="20,0,0", Scale="1,1,1") {
    rotationMode=BMM_Additive
    rotationSpace=BCS_BoneSpace
    translationSpace=BCS_BoneSpace
    alphaBoolBlend {blendOption=Linear}
    alphaScaleBiasClamp {interpSpeedDecreasing=2, interpSpeedIncreasing=2}
    boneToModify {boneName=HeadBand_MSTR}
    rotation {pitch=20}
   }
   n7 = Blend Poses by bool(
    BlendPose_0=Component To Local(ComponentPose=@n3)
    BlendPose_1=Component To Local(ComponentPose=@n6)
    bActiveValue=Get QuietPeanuts Backbling Alpha() -> (QuietPeanuts Backbling Alpha:bool)
   ) {blendType=Linear}
   n8 = Control Rig(Source=Copy Pose From Mesh() {bUseAttachedParent=true}) {
    alphaBoolBlend {blendOption=Linear}
    controlRigAssetReference {blueprintRigClass=&M_MED_QuietPeanuts_FaceAcc_CtrlRig_C}
   }
   n9 = "Transform (Modify) Bone - Bone: dyn_HeadBand_l_MSTR"(
    ComponentPose=Local To Component(LocalPose=@n8)
    Alpha=Get Tassle 01 Alpha() -> (Tassle 01 Alpha:double)
    Translation="0,0,0"
    Rotation=Get TesselRotator_01() -> (TesselRotator_01:Rotator)
    Scale="1,1,1"
   ) {
    rotationMode=BMM_Additive
    rotationSpace=BCS_BoneSpace
    translationSpace=BCS_BoneSpace
    alphaBoolBlend {blendOption=Linear}
    alphaScaleBiasClamp {interpSpeedDecreasing=2, interpSpeedIncreasing=2}
    boneToModify {boneName=dyn_HeadBand_l_MSTR}
   }
   n10 = "Transform (Modify) Bone - Bone: dyn_HeadBand_r_MSTR"(
    ComponentPose=@n9
    Alpha=Get Tassle 02 Alpha() -> (Tassle 02 Alpha:double)
    Translation="0,0,0"
    Rotation=Get TesselRotator_02() -> (TesselRotator_02:Rotator)
    Scale="1,1,1"
   ) {
    rotationMode=BMM_Additive
    rotationSpace=BCS_BoneSpace
    translationSpace=BCS_BoneSpace
    alphaBoolBlend {blendOption=Linear}
    alphaScaleBiasClamp {bInterpResult=true, interpSpeedDecreasing=2, interpSpeedIncreasing=2}
    boneToModify {boneName=dyn_HeadBand_r_MSTR}
   }
   n11 = RigidBody(ComponentPose=@n10, OverrideWorldGravity=Get GravityOverride_tessel()) {
    bClampLinearTranslationLimitToRefPose=true
    bOverrideWorldGravity=true
    overridePhysicsAsset=&M_MED_QuietPeanuts_FaceAcc_Backbling_Physics
    simulationSpace=BaseBoneSpace
    alphaBoolBlend {blendOption=Linear}
    baseBoneRef {boneName=neck_01}
    componentAppliedLinearAccClamp=<3000 3000 3000>
    componentLinearAccScale {x=0.1}
    componentLinearVelScale {x=0.1}
    overrideWorldGravity {z=-980}
    simSpaceSettings {
     maxAngularAcceleration=1000
     maxAngularVelocity=3
     maxLinearAcceleration=700
     maxLinearVelocity=500
     velocityScaleZ=0
     worldAlpha=1
     externalLinearDragV=<0.5 0.5>
    }
   }
   n12 = Control Rig(Source=Copy Pose From Mesh() {bUseAttachedParent=true}) {
    alphaBoolBlend {blendOption=Linear}
    controlRigAssetReference {blueprintRigClass=&M_MED_QuietPeanuts_FaceAcc_CtrlRig_C}
   }
   n13 = "Transform (Modify) Bone - Bone: dyn_HeadBand_l_MSTR"(
    ComponentPose=Local To Component(LocalPose=@n12)
    Alpha=Get Tassle 01 Alpha()
    Translation="0,0,0"
    Rotation=Get TesselRotator_01()
    Scale="1,1,1"
   ) {
    rotationMode=BMM_Additive
    rotationSpace=BCS_BoneSpace
    translationSpace=BCS_BoneSpace
    alphaBoolBlend {blendOption=Linear}
    alphaScaleBiasClamp {interpSpeedDecreasing=2, interpSpeedIncreasing=2}
    boneToModify {boneName=dyn_HeadBand_l_MSTR}
   }
   n14 = "Transform (Modify) Bone - Bone: dyn_HeadBand_r_MSTR"(ComponentPose=@n13, Alpha=Get Tassle 02 Alpha(), Translation="0,0,0", Rotation=Get TesselRotator_02(), Scale="1,1,1") {
    rotationMode=BMM_Additive
    rotationSpace=BCS_BoneSpace
    translationSpace=BCS_BoneSpace
    alphaBoolBlend {blendOption=Linear}
    alphaScaleBiasClamp {bInterpResult=true, interpSpeedDecreasing=2, interpSpeedIncreasing=2}
    boneToModify {boneName=dyn_HeadBand_r_MSTR}
   }
   n15 = RigidBody(ComponentPose=@n14, OverrideWorldGravity=Get GravityOverride_tessel()) {
    bClampLinearTranslationLimitToRefPose=true
    bOverrideWorldGravity=true
    overridePhysicsAsset=&/M_MED_QuietPeanuts_FaceAcc_Ground_Physics
    simulationSpace=BaseBoneSpace
    alphaBoolBlend {blendOption=Linear}
    baseBoneRef {boneName=neck_01}
    componentAppliedLinearAccClamp=<3000 3000 3000>
    componentLinearAccScale {x=0.1}
    componentLinearVelScale {x=0.1}
    overrideWorldGravity {z=-980}
    simSpaceSettings {
     maxAngularAcceleration=1000
     maxAngularVelocity=3
     maxLinearAcceleration=700
     maxLinearVelocity=500
     velocityScaleZ=0.15
     worldAlpha=1
     externalLinearDragV=<0.5 0.5>
    }
   }
   n16 = Blend Poses by bool(
    BlendPose_0=Component To Local(ComponentPose=@n11)
    BlendPose_1=Component To Local(ComponentPose=@n15)
    bActiveValue=Get QuietPeanuts Backbling Collision Alpha() -> (QuietPeanuts Backbling Collision Alpha:bool)
   ) {blendType=Linear}
   n17 = Get Is Skydive Bool Alpha() -> (Is Skydive Bool Alpha:bool)
   n18 = Control Rig(Source=Copy Pose From Mesh() {bUseAttachedParent=true}) {
    alphaBoolBlend {blendOption=Linear}
    controlRigAssetReference {blueprintRigClass=&M_MED_QuietPeanuts_FaceAcc_CtrlRig_C}
   }
   n19 = "Transform (Modify) Bone - Bone: dyn_HeadBand_l_MSTR"(
    ComponentPose=Local To Component(LocalPose=@n18)
    Alpha=Get Tassle 01 Alpha()
    Translation="0,0,0"
    Rotation=Get TesselRotator_01()
    Scale="1,1,1"
   ) {
    rotationMode=BMM_Additive
    rotationSpace=BCS_BoneSpace
    translationSpace=BCS_BoneSpace
    alphaBoolBlend {blendOption=Linear}
    alphaScaleBiasClamp {interpSpeedDecreasing=2, interpSpeedIncreasing=2}
    boneToModify {boneName=dyn_HeadBand_l_MSTR}
   }
   n20 = "Transform (Modify) Bone - Bone: dyn_HeadBand_r_MSTR"(ComponentPose=@n19, Alpha=Get Tassle 02 Alpha(), Translation="0,0,0", Rotation=Get TesselRotator_02(), Scale="1,1,1") {
    rotationMode=BMM_Additive
    rotationSpace=BCS_BoneSpace
    translationSpace=BCS_BoneSpace
    alphaBoolBlend {blendOption=Linear}
    alphaScaleBiasClamp {bInterpResult=true, interpSpeedDecreasing=2, interpSpeedIncreasing=2}
    boneToModify {boneName=dyn_HeadBand_r_MSTR}
   }
   n21 = "Transform (Modify) Bone - Bone: HeadBand_MSTR"(ComponentPose=@n20, Alpha=1, Translation="0,0,0", Rotation="25,0,0", Scale="1,1,1") {
    rotationMode=BMM_Additive
    rotationSpace=BCS_BoneSpace
    translationSpace=BCS_BoneSpace
    alphaBoolBlend {blendOption=Linear}
    alphaScaleBiasClamp {bInterpResult=true, interpSpeedDecreasing=2, interpSpeedIncreasing=2}
    boneToModify {boneName=HeadBand_MSTR}
    rotation {pitch=25}
   }
   n22 = Anim Dynamics - Bone: dyn_HeadBand_l_01_jnt(ComponentPose=@n21, Alpha=1, ExternalForce="0,0,0") {
    angularDampingOverride=1
    angularSpringConstant=200
    bAngularSpring=true
    bOverrideAngularDamping=true
    bUseGravityOverride=true
    simulationSpace=RootRelative
    alphaBoolBlend {blendOption=Linear}
    boundBone {boneName=dyn_HeadBand_l_01_jnt}
    componentAppliedLinearAccClamp=<6000 6000 6000>
    componentLinearAccScale=<0.5 0.5 0.5>
    componentLinearVelScale=<0.5 0.5 0.5>
    gravityOverride {z=-450}
    physicsBodyDefinitions [{
      boundBone {boneName=dyn_HeadBand_l_01_jnt}
      boxExtents=<10 10 10>
      constraintSetup {angularLimitsMax=<2 2 25>, angularLimitsMin=<-2 -45 -2>, angularTarget=<1 0 0>}
      localJointOffset=<12 0 0>
     }]
   }
   n23 = Anim Dynamics - Bone: dyn_HeadBand_l_02_jnt(ComponentPose=@n22, Alpha=1, ExternalForce="0,0,0") {
    angularDampingOverride=1
    angularSpringConstant=200
    bAngularSpring=true
    bOverrideAngularDamping=true
    bUseGravityOverride=true
    simulationSpace=RootRelative
    alphaBoolBlend {blendOption=Linear}
    boundBone {boneName=dyn_HeadBand_l_02_jnt}
    componentAppliedLinearAccClamp=<6000 6000 6000>
    componentLinearAccScale=<0.5 0.5 0.5>
    componentLinearVelScale=<0.5 0.5 0.5>
    gravityOverride {z=-450}
    physicsBodyDefinitions [{
      boundBone {boneName=dyn_HeadBand_l_02_jnt}
      boxExtents=<10 10 10>
      constraintSetup {angularLimitsMax=<2 2 25>, angularLimitsMin=<-2 -45 -2>, angularTarget=<1 0 0>}
      localJointOffset=<12 0 0>
     }]
   }
   n24 = Anim Dynamics - Bone: dyn_HeadBand_l_03_jnt(ComponentPose=@n23, Alpha=1, ExternalForce="0,0,0") {
    angularDampingOverride=1
    angularSpringConstant=200
    bAngularSpring=true
    bOverrideAngularDamping=true
    bUseGravityOverride=true
    simulationSpace=RootRelative
    alphaBoolBlend {blendOption=Linear}
    boundBone {boneName=dyn_HeadBand_l_03_jnt}
    componentAppliedLinearAccClamp=<6000 6000 6000>
    componentLinearAccScale=<0.5 0.5 0.5>
    componentLinearVelScale=<0.5 0.5 0.5>
    gravityOverride {z=-450}
    physicsBodyDefinitions [{
      boundBone {boneName=dyn_HeadBand_l_03_jnt}
      boxExtents=<10 10 10>
      constraintSetup {angularLimitsMax=<2 2 25>, angularLimitsMin=<-2 -45 -2>, angularTarget=<1 0 0>}
      localJointOffset=<12 0 0>
     }]
   }
   n25 = Anim Dynamics - Bone: dyn_HeadBand_l_04_jnt(ComponentPose=@n24, Alpha=1, ExternalForce="0,0,0") {
    angularDampingOverride=1
    angularSpringConstant=200
    bAngularSpring=true
    bOverrideAngularDamping=true
    bUseGravityOverride=true
    simulationSpace=RootRelative
    alphaBoolBlend {blendOption=Linear}
    boundBone {boneName=dyn_HeadBand_l_04_jnt}
    componentAppliedLinearAccClamp=<6000 6000 6000>
    componentLinearAccScale=<0.5 0.5 0.5>
    componentLinearVelScale=<0.5 0.5 0.5>
    gravityOverride {z=-450}
    physicsBodyDefinitions [{
      boundBone {boneName=dyn_HeadBand_l_04_jnt}
      boxExtents=<10 10 10>
      constraintSetup {angularLimitsMax=<2 2 25>, angularLimitsMin=<-2 -45 -2>, angularTarget=<1 0 0>}
      localJointOffset=<12 0 0>
     }]
   }
   n26 = Anim Dynamics - Bone: dyn_HeadBand_l_05_jnt(ComponentPose=@n25, Alpha=1, ExternalForce="0,0,0") {
    angularDampingOverride=1
    angularSpringConstant=200
    bAngularSpring=true
    bOverrideAngularDamping=true
    bUseGravityOverride=true
    simulationSpace=RootRelative
    alphaBoolBlend {blendOption=Linear}
    boundBone {boneName=dyn_HeadBand_l_05_jnt}
    componentAppliedLinearAccClamp=<6000 6000 6000>
    componentLinearAccScale=<0.5 0.5 0.5>
    componentLinearVelScale=<0.5 0.5 0.5>
    gravityOverride {z=-450}
    physicsBodyDefinitions [{
      boundBone {boneName=dyn_HeadBand_l_05_jnt}
      boxExtents=<10 10 10>
      constraintSetup {angularLimitsMax=<2 2 25>, angularLimitsMin=<-2 -45 -2>, angularTarget=<1 0 0>}
      localJointOffset=<12 0 0>
     }]
   }
   n27 = Anim Dynamics - Bone: dyn_HeadBand_r_01_jnt(ComponentPose=@n26, Alpha=1, ExternalForce="0,0,0") {
    angularDampingOverride=1
    angularSpringConstant=200
    bAngularSpring=true
    bOverrideAngularDamping=true
    bUseGravityOverride=true
    simulationSpace=RootRelative
    alphaBoolBlend {blendOption=Linear}
    boundBone {boneName=dyn_HeadBand_r_01_jnt}
    componentAppliedLinearAccClamp=<6000 6000 6000>
    componentLinearAccScale=<0.5 0.5 0.5>
    componentLinearVelScale=<0.5 0.5 0.5>
    gravityOverride {z=-450}
    physicsBodyDefinitions [{
      boundBone {boneName=dyn_HeadBand_r_01_jnt}
      boxExtents=<10 10 10>
      constraintSetup {angularLimitsMax=<2 2 2>, angularLimitsMin=<-2 -45 -25>, angularTarget=<1 0 0>}
      localJointOffset=<12 0 0>
     }]
   }
   n28 = Anim Dynamics - Bone: dyn_HeadBand_r_02_jnt(ComponentPose=@n27, Alpha=1, ExternalForce="0,0,0") {
    angularDampingOverride=1
    angularSpringConstant=200
    bAngularSpring=true
    bOverrideAngularDamping=true
    bUseGravityOverride=true
    simulationSpace=RootRelative
    alphaBoolBlend {blendOption=Linear}
    boundBone {boneName=dyn_HeadBand_r_02_jnt}
    componentAppliedLinearAccClamp=<6000 6000 6000>
    componentLinearAccScale=<0.5 0.5 0.5>
    componentLinearVelScale=<0.5 0.5 0.5>
    gravityOverride {z=-450}
    physicsBodyDefinitions [{
      boundBone {boneName=dyn_HeadBand_r_02_jnt}
      boxExtents=<10 10 10>
      constraintSetup {angularLimitsMax=<2 2 2>, angularLimitsMin=<-2 -45 -25>, angularTarget=<1 0 0>}
      localJointOffset=<12 0 0>
     }]
   }
   n29 = Anim Dynamics - Bone: dyn_HeadBand_r_03_jnt(ComponentPose=@n28, Alpha=1, ExternalForce="0,0,0") {
    angularDampingOverride=1
    angularSpringConstant=200
    bAngularSpring=true
    bOverrideAngularDamping=true
    bUseGravityOverride=true
    simulationSpace=RootRelative
    alphaBoolBlend {blendOption=Linear}
    boundBone {boneName=dyn_HeadBand_r_03_jnt}
    componentAppliedLinearAccClamp=<6000 6000 6000>
    componentLinearAccScale=<0.5 0.5 0.5>
    componentLinearVelScale=<0.5 0.5 0.5>
    gravityOverride {z=-450}
    physicsBodyDefinitions [{
      boundBone {boneName=dyn_HeadBand_r_03_jnt}
      boxExtents=<10 10 10>
      constraintSetup {angularLimitsMax=<2 2 2>, angularLimitsMin=<-2 -45 -25>, angularTarget=<1 0 0>}
      localJointOffset=<12 0 0>
     }]
   }
   n30 = Anim Dynamics - Bone: dyn_HeadBand_r_04_jnt(ComponentPose=@n29, Alpha=1, ExternalForce="0,0,0") {
    angularDampingOverride=1
    angularSpringConstant=200
    bAngularSpring=true
    bOverrideAngularDamping=true
    bUseGravityOverride=true
    simulationSpace=RootRelative
    alphaBoolBlend {blendOption=Linear}
    boundBone {boneName=dyn_HeadBand_r_04_jnt}
    componentAppliedLinearAccClamp=<6000 6000 6000>
    componentLinearAccScale=<0.5 0.5 0.5>
    componentLinearVelScale=<0.5 0.5 0.5>
    gravityOverride {z=-450}
    physicsBodyDefinitions [{
      boundBone {boneName=dyn_HeadBand_r_04_jnt}
      boxExtents=<10 10 10>
      constraintSetup {angularLimitsMax=<2 2 2>, angularLimitsMin=<-2 -45 -25>, angularTarget=<1 0 0>}
      localJointOffset=<12 0 0>
     }]
   }
   n31 = Anim Dynamics - Bone: dyn_HeadBand_r_05_jnt(ComponentPose=@n30, Alpha=1, ExternalForce="0,0,0") {
    angularDampingOverride=1
    angularSpringConstant=200
    bAngularSpring=true
    bOverrideAngularDamping=true
    bUseGravityOverride=true
    simulationSpace=RootRelative
    alphaBoolBlend {blendOption=Linear}
    boundBone {boneName=dyn_HeadBand_r_05_jnt}
    componentAppliedLinearAccClamp=<6000 6000 6000>
    componentLinearAccScale=<0.5 0.5 0.5>
    componentLinearVelScale=<0.5 0.5 0.5>
    gravityOverride {z=-450}
    physicsBodyDefinitions [{
      boundBone {boneName=dyn_HeadBand_r_05_jnt}
      boxExtents=<10 10 10>
      constraintSetup {angularLimitsMax=<2 2 2>, angularLimitsMin=<-2 -45 -25>, angularTarget=<1 0 0>}
      localJointOffset=<12 0 0>
     }]
   }
   Output Pose(Result=Blend Poses by bool(
     BlendPose_0=Blend Poses by bool(BlendPose_0=@n7, BlendPose_1=@n16, bActiveValue=@n17) {blendType=Linear}
     BlendPose_1=Component To Local(ComponentPose=@n31)
     bActiveValue=Get bRigidBodyEnabled()
    ) {blendType=Linear}) {layerGroup=None}
   Get bIsSkyDiving()
  ]}
 EventGraph {nodes [
   Event BlueprintUpdateAnimation()
   > n1 = Sequence!(execute) -> (then_0, then_1, then_2:Exec, then_3:Exec, then_4:Exec)
   n2 = OR Boolean!(A=Get bIsSkydiveFloating(), B=Get bIsSkydiveDiveMode(), C:bool=Get bIsParachuteOpened()) -> (ReturnValue)
   > Set Is Skydiving Alpha(Is Skydiving Alpha:double="To Float (Boolean)"(InBool=@n2)) -> (Output_Get:double)
   > Set Is Skydive Bool Alpha(Is Skydive Bool Alpha:bool=@n2) -> (Output_Get:bool)
   n3 = Gravity Override State Function DEPRECATED(
    execute=@n1.then_1
    CharacterPartAnimInstance=Self-Reference()
    GameTimeInSeconds=GetGameTimeInSeconds()
    GravityOverrideParams=Get Gravity Override Params()
   )
   > Set GravityOverride_tessel(GravityOverride_tessel:Vector=@n3) -> (Output_Get:Vector)
   n4 = float * float(A=GetGameTimeInSeconds(), B=1.125) {is=title K2Node_PromotableOperator}
   n5 = RandomRotatorFunction!(
    execute=@n1.then_2
    GameTimeInSeconds:double=@n4
    PerlinScale:double=1
    Frequency:Rotator="5,2.8,0"
    Amplitude:Rotator="30,15,0"
    Offset:Rotator="30,-1.8,0"
    TimeOffset:double=0
   ) -> (then, RandomRotator:Rotator)
   > Set TesselRotator_01(TesselRotator_01:Rotator=@n5) -> (Output_Get:Rotator)
   n6 = RandomRotatorFunction(execute=@n1.then_3, GameTimeInSeconds=@n4, PerlinScale=1, Frequency="4.5,1.6,0", Amplitude="30,15,0", Offset="35,15,0", TimeOffset=0)
   > Set TesselRotator_02(TesselRotator_02:Rotator=@n6) -> (Output_Get:Rotator)
   n7 = Branch(execute=@n1.then_4, Condition=Get bIsInFrontEnd())
   > Set Tassle 01 Alpha(Tassle 01 Alpha:double=0.25) -> (Output_Get:double)
   > Set Tassle 02 Alpha(Tassle 02 Alpha:double=0.425) -> (Output_Get:double)
   Set Tassle 01 Alpha(execute=@n7.else, Tassle 01 Alpha=0.5)
   > Set Tassle 02 Alpha(Tassle 02 Alpha=0.85)
   n8 = Event OnWeaponChanged()
   Get Dynamic Material Instance(execute=[@n8.then, @n14.then], SkeletalMeshComponent=GetPartSkeletalMeshComponent())
   > n9 = Sequence!(execute) -> (then_0, then_1)
   n10 = BackpackItemDefinitionHasAnyMetaTag!(Tags:Array of Gameplay Tag Structures=Make Array("[0]"=(TagName="Cosmetics.MetaTags.QuietPeanuts")), bRequiresEquipped:bool=false) -> (ReturnValue:bool)
   n11 = PickaxeItemDefinitionHasAnyMetaTag(Tags=Make Array("[0]"=(TagName="Cosmetics.MetaTags.QuietPeanuts")), bRequiresEquipped=true)
   > n12 = Branch(Condition=AND Boolean(A=NOT Boolean(A=@n11), B=@n10))
   > Set QuietPeanuts Backbling Collision Alpha(QuietPeanuts Backbling Collision Alpha:bool=true) -> (Output_Get:bool)
   Set QuietPeanuts Backbling Collision Alpha(execute=@n12.else, QuietPeanuts Backbling Collision Alpha=false)
   n13 = Branch(execute=@n9.then_1, Condition=@n10)
   > Set QuietPeanuts Backbling Alpha(QuietPeanuts Backbling Alpha:bool=true) -> (Output_Get:bool)
   Set QuietPeanuts Backbling Alpha(execute=@n13.else, QuietPeanuts Backbling Alpha=false)
   n14 = Event OnCharacterCustomizationLoaded()
   Get bIsSkyDiving()
  ]}
]
variables [
 Gravity Override Params {
  type=GravityOverrideParamsStruct
  default {
   emoteOrMelee {gravityOverride=<0 -50 -15>, windAmplitude=<50 -25 100>, windFrequency=<1.5 10 4>, windOffset=<0 -100 -50>}
   front_End {gravityOverride=<0 -50 -15>, windAmplitude=<50 -25 100>, windFrequency=<1.5 10 4>, windOffset=<0 -100 -50>}
   onGround_Moving {gravityOverride=<0 -50 -15>, windAmplitude=<50 -25 100>, windFrequency=<1.8 13 6>, windOffset=<0 -100 -50>}
   onGround_Standing {gravityOverride=<0 -50 -15>, windAmplitude=<50 -25 100>, windFrequency=<1.8 13 6>, windOffset=<0 -100 -50>}
  }
 }
 GravityOverride_tessel {type=Vector}
 Is Skydive Bool Alpha {type=boolean}
 Is Skydiving Alpha {type=number}
 QuietPeanuts Backbling Alpha {type=boolean}
 QuietPeanuts Backbling Collision Alpha {type=boolean}
 Tassle 01 Alpha {type=number}
 Tassle 02 Alpha {type=number}
 TesselRotator_01 {type=Rotator}
 TesselRotator_02 {type=Rotator}
]
